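# Unlit, masked, two-sided aquarium master material for Niagara sprites, ribbons and mesh particles, defaulting to a seaweed texture.
# A Water? switch selects two multiplied panning caustic layers or a channel-selectable texture, tinted by particle color, with an optional animated sine band added to emissive.
# A noise step driven by dynamic Dissolve cuts alpha masks.
# Optional WPO pans a gradient texture for waves, reusing dynamic-parameter slots for speed and UVs.
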Material /DealershipSharedLibrary/Assets/Sets/Aquarium/Materials/M_Aquarium_Master
graph {
 properties {
  bAutomaticallySetUsageInEditor=false
  bUsedWithNiagaraMeshParticles=true
  bUsedWithNiagaraRibbons=true
  bUsedWithNiagaraSprites=true
  blendMode=BLEND_Masked
  shadingModel=MSM_Unlit
  twoSided=true
  usedShadingModels=Unlit
 }
 nodes [
  n1 = ScalarParameter() {parameterName=Panner}
  n2 = ScalarParameter() {defaultValue=1, parameterName=No_Panner}
  n3 = Constant() {r=0.25}
  n4 = Constant() {r=1}
  n5 = Frac(None=Multiply(A=TextureCoordinate(), B=AppendVector(A=@n3, B=@n4)))
  n6 = Sine(None=Add(A=ComponentMask(None=@n5) {g=true}, B=Time())) {period=0.5}
  n7 = Clamp(None=Multiply(A=ConstantBiasScale(None=@n6) {bias=-0.1, scale=0.8}, B=ScalarParameter() {defaultValue=1, parameterName=Power_Panner}))
  n8 = Constant()
  n9 = TextureCoordinate()
  n10 = AppendVector(A=ScalarParameter() {defaultValue=1, parameterName=Caustic_X_1}, B=ScalarParameter() {defaultValue=1, parameterName=Caustic_Y_1})
  n11 = ScalarParameter() {defaultValue=0.05, parameterName=Speed_X_1}
  n12 = ScalarParameter() {defaultValue=0.05, parameterName=Speed_Y_1}
  n13 = TextureSampleParameter2D(UVs=Panner(Coordinate=Multiply(A=@n9, B=@n10), Speed=AppendVector(A=@n11, B=@n12)) {speedX=0.05, speedY=0.05}) {parameterName=Caustic_1, texture=/DealershipSharedLibrary/FX/Textures/Noises/T_Murk_CausticsNoMip}
  n14 = TextureCoordinate() {uTiling=2, vTiling=2}
  n15 = AppendVector(A=ScalarParameter() {defaultValue=2, parameterName=Caustic_X_2}, B=ScalarParameter() {defaultValue=2, parameterName=Caustic_Y_2})
  n16 = ScalarParameter() {defaultValue=-0.04, parameterName=Speed_X_2}
  n17 = ScalarParameter() {defaultValue=-0.03, parameterName=Speed_Y_2}
  n18 = TextureSampleParameter2D(UVs=Panner(Coordinate=Multiply(A=@n14, B=@n15), Speed=AppendVector(A=@n16, B=@n17)) {speedX=-0.04, speedY=-0.03}) {parameterName=Caustic_2, texture=&T_Murk_CausticsNoMip}
  n19 = Multiply(
   A=Multiply(A=ScalarParameter() {defaultValue=1, parameterName=Power_Caustic_1}, B=@n13.R)
   B=Multiply(A=@n18.R, B=ScalarParameter() {defaultValue=1, parameterName=Power_Caustic_2})
  )
  n20 = DynamicParameter() {defaultValue=<rgba 1 0 8 1>, paramNames [Dissolve, "rgb? b&&w?", Caustics, Param4]}
  MaterialExpressionTextureSampleParameter2D_0 = TextureSampleParameter2D() {parameterName=MainTexture, texture=/DealershipSharedLibrary/Assets/Sets/Aquarium/Textures/T_Aquarium_Seaweed}
  n21 = Switch!(
   SwitchValue=@n20."rgb? b&&w?"
   Default
   RGB=@MaterialExpressionTextureSampleParameter2D_0.RGB
   R=@MaterialExpressionTextureSampleParameter2D_0.RGB
   G=@MaterialExpressionTextureSampleParameter2D_0.RGB
   B=@MaterialExpressionTextureSampleParameter2D_0.RGB
  ) {inputs [
    {
     inputName=RGB
     input {
      expression=&self:MaterialExpressionTextureSampleParameter2D_0
      inputName=None
      mask=1
      maskA=0
      maskB=1
      maskG=1
      maskR=1
      outputIndex=0
     }
    }
    {
     inputName=R
     input {
      expression=&self:MaterialExpressionTextureSampleParameter2D_0
      inputName=None
      mask=1
      maskA=0
      maskB=0
      maskG=0
      maskR=1
      outputIndex=1
     }
    }
    {
     inputName=G
     input {
      expression=&self:MaterialExpressionTextureSampleParameter2D_0
      inputName=None
      mask=1
      maskA=0
      maskB=0
      maskG=1
      maskR=0
      outputIndex=2
     }
    }
    {
     inputName=B
     input {
      expression=&self:MaterialExpressionTextureSampleParameter2D_0
      inputName=None
      mask=1
      maskA=0
      maskB=1
      maskG=0
      maskR=0
      outputIndex=3
     }
    }
   ]}
  n22 = StaticSwitchParameter(True=Add(A=Constant3Vector() {constant=<rgba 0.1354 0.1354 0.1354 0>}, B=Multiply(A=@n19, B=@n20.Caustics) {constB=8}), False=@n21) {parameterName=Water?}
  n23 = ParticleColor()
  MaterialFunctionCall(Emissive=Add(A=If(A=@n1, B=@n2, "A > B"=@n7, "A < B"=@n8), B=Multiply(A=@n23.RGB, B=@n22))) {materialFunction=&MF_FX_EmissiveAdjustments}
  MaterialFunctionCall() {materialFunction=&Lerp_3Color}
  n24 = ComponentMask(None=TextureCoordinate()) {g=true}
  n25 = ScalarParameter() {defaultValue=10, parameterName=Power_Mask}
  n26 = Multiply(A=Multiply(A=@n13.A, B=@n18.A), B=@n23.A)
  n27 = Step(Y=Add(
    A=TextureSampleParameter2D() {parameterName=Dissolve, texture=/DealershipSharedLibrary/FX/Textures/Noises/T_DLR_CloudDistort_A}.R
    B=@n20.Dissolve
   ))
  Reroute(None=Multiply(A=Multiply(A=@n24, B=@n25), B=Subtract(A=@n26, B=@n27)))
  n28 = VertexColor()
  n29 = Multiply(
   A=Add(A=@n28.A, B=TextureCoordinate() {coordinateIndex=1})
   B=ComponentMask(None=VectorParameter() {group=WPO, parameterName=TextureWPO_UVtiling, defaultValue=<rgba 1 0 0 0>}.RGB) {g=true, r=true}
  )
  n30 = DynamicParameter() {defaultValue=<rgba 1 0 8 1>, paramNames [Dissolve, "rgb? b&&w?", Caustics, Param4]}
  n31 = LinearInterpolate(A=@n29, B=Multiply(A=@n29, B=@n30."rgb? b&&w?"), Alpha=ScalarParameter() {group=WPO, parameterName=TextureWPO_UVTiling_Per_Particle})
  n32 = Add(A=@n31, B=@n30.Caustics)
  n33 = ScalarParameter() {group=WPO, parameterName=TextureWPO_UVSpeed_IntegrateFromNiagara}
  n34 = VectorParameter() {group=WPO, parameterName=TextureWPO_Speed, defaultValue=<rgba 1 0 0 0>}
  n35 = AppendVector(
   A=Add(A=ComponentMask(None=@n34.RGB) {r=true}, B=LinearInterpolate(Alpha=@n28.A) {constA=0.15, constB=0.35})
   B=ComponentMask(None=@n34.RGB) {g=true}
  )
  n36 = Multiply(A=@n35, B=@n30.Dissolve)
  n37 = ScalarParameter() {group=WPO, parameterName=TextureWPO_Per_Particle}
  n38 = TextureSample(UVs=Panner(Coordinate=LinearInterpolate(A=@n31, B=@n32, Alpha=@n33), Speed=LinearInterpolate(A=@n35, B=@n36, Alpha=@n37)) {speedX=0.15}) {texture=/Game/Global/Textures/Default/Masks_and_Gradients/WeaponGradients/T_Gradient_DesertSky_Color}
  n39 = Multiply(A=@n38.R, B=ScalarParameter() {defaultValue=1, group=WPO, parameterName=WPO_HeightX}) {constB=0.5}
  n40 = Multiply(A=@n38.R, B=ScalarParameter() {defaultValue=1, group=WPO, parameterName=WPO_HeightY}) {constB=0.5}
  n41 = ScalarParameter() {defaultValue=1, group=WPO, parameterName=WPO_HeightZ}
  n42 = LinearInterpolate(
   A=Constant3Vector()
   B=AppendVector(A=AppendVector(A=@n39, B=@n40), B=Multiply(A=@n38.R, B=@n41) {constB=0.6})
   Alpha=LinearInterpolate(B=ScalarParameter() {group=WPO, parameterName=WPO_EyeMovent}, Alpha=VertexColor().R) {constA=1, constB=0}
  )
  n43 = LinearInterpolate(
   A=ScalarParameter() {defaultValue=1, group=WPO, parameterName=WPO_Wave_Intensity_Close}
   B=ScalarParameter() {defaultValue=1, group=WPO, parameterName=WPO_Wave_Intenstiy}
  )
  StaticSwitchParameter(True=Transform(None=Multiply(A=@n42, B=@n43) {constB=80}) {transformSourceType=TRANSFORMSOURCE_Local}, False=Constant()) {parameterName=WPO?}
  Subtract(A=Multiply(A=@n23.A, B=@MaterialExpressionTextureSampleParameter2D_0.A), B=@n27)
 ]
}
